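# Sound cue for the active state of the Zapper Trap V2 in Athena, layering a one-shot ready sound with a looping active sound.
# A mixer plays one of two reused electric wall trap ready sounds immediately, alongside one of three Zapper loops started after a short fixed delay.
# Each layer has its own random modulation.
# It uses Athena trap attenuation and overrides concurrency to prevent new instances at the limit.
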SoundCue /Game/Athena/Sounds/Items/Zapper_Trap_V2/Zapper_Trap_V2_Active_Cue
root=Mixer_1
nodes [
 {id=Delay_0, children [Random_0], properties {delayMax=0.4, delayMin=0.4}}
 {id=Mixer_1, children [Modulator_10,Modulator_0], properties {inputVolume [1,1]}}
 {id=Modulator_0, children [Delay_0], properties {pitchMin=0.9}}
 {id=Modulator_10, children [Random_20], properties {pitchMax=1.1}}
 {id=Random_0, children [WavePlayer_0,WavePlayer_1,WavePlayer_2], properties {weights [1,1,1]}}
 {id=Random_20, children [WavePlayer_71, WavePlayer_72], properties {weights [1,1]}}
 {id=WavePlayer_0, properties {bLooping=true, soundWaveAssetPtr=&/WPN_Trap_Zapper_V2_Active_Loop_01}}
 {id=WavePlayer_1, properties {bLooping=true, soundWaveAssetPtr=&/WPN_Trap_Zapper_V2_Active_Loop_02}}
 {id=WavePlayer_2, properties {bLooping=true, soundWaveAssetPtr=&/WPN_Trap_Zapper_V2_Active_Loop_03}}
 {id=WavePlayer_71, properties {soundWaveAssetPtr=/Game/Sounds/Fort_Traps/Wall_Electric/WallTrap_Electric_Ready_001}}
 {id=WavePlayer_72, properties {soundWaveAssetPtr=/Game/Sounds/Fort_Traps/Wall_Electric/WallTrap_Electric_Ready_002}}
]
properties {
 attenuationSettings=/Game/Sounds/Attenuation/Fort_Traps_Athena_Att
 bOverrideConcurrency=true
 maxDistance=1200
 soundClassObject=&Fort_Traps
 volumeMultiplier=0.5
 attenuationOverrides {falloffDistance=2100}
 concurrencyOverrides {resolutionRule=PreventNew}
}
