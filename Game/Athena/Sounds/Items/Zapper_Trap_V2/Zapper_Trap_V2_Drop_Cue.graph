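# Sound cue for dropping the Zapper Trap V2 item, played as loot drop audio with weapon-drop attenuation.
# A single drop wave feeds a Modulator root at reduced cue volume, routed to the loot-collect sound class.
# A one-per-owner concurrency set is assigned alongside override settings that stop the oldest instance.
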SoundCue /Game/Athena/Sounds/Items/Zapper_Trap_V2/Zapper_Trap_V2_Drop_Cue
root=Modulator_0
nodes [
 {id=Modulator_0, children [WavePlayer_1]}
 {id=WavePlayer_1, properties {soundWaveAssetPtr=&/WPN_Trap_Zapper_V2_Drop_01}}
]
properties {
 attenuationSettings=&Fort_Weapon_Drop_Att
 maxDistance=3000
 priority=0.5
 soundClassObject=&Fort_Loot_Collect
 volumeMultiplier=0.498
 concurrencyOverrides {maxCount=2, resolutionRule=StopOldest}
 concurrencySet [/Game/Sounds/Concurrency_Rules/OneMaximum_Owner_StopLowestPriority]
}
