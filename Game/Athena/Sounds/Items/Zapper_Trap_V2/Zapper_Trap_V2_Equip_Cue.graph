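# Sound cue for the equip sound of the Zapper Trap V2 item in Athena.
# A modulator feeds from a single wave player for the V2 zapper equip wave, giving it slight variation on each play.
# It routes to the traps sound class and uses weapon reload attenuation at reduced volume.
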SoundCue /Game/Athena/Sounds/Items/Zapper_Trap_V2/Zapper_Trap_V2_Equip_Cue
root=Modulator_0
nodes [
 {id=Modulator_0, children [WavePlayer_0]}
 {id=WavePlayer_0, properties {soundWaveAssetPtr=&/WPN_Trap_Zapper_V2_Equip_01}}
]
properties {
 attenuationSettings=&Fort_Weapon_Reload_Att
 maxDistance=2250
 soundClassObject=&Fort_Traps
 volumeMultiplier=0.6
}
