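# Sound cue for the explosion of the Athena Zapper Trap V2 item.
# It plays a single Zapper V2 fire wave through a Modulator that randomizes pitch slightly around normal.
# It uses the shared trap attenuation and sound class at reduced volume, with a concurrency override that prevents new instances once the limit is reached.
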SoundCue /Game/Athena/Sounds/Items/Zapper_Trap_V2/Zapper_Trap_V2_Explode_Cue
root=Modulator_4
nodes [
 {id=Modulator_4, children [WavePlayer_0], properties {pitchMax=1.1, pitchMin=0.9}}
 {id=WavePlayer_0, properties {soundWaveAssetPtr=&/WPN_Trap_Zapper_V2_Fire_05}}
]
properties {
 attenuationSettings=&Fort_Traps_Att
 bOverrideConcurrency=true
 maxDistance=6000
 soundClassObject=&Fort_Traps
 volumeMultiplier=0.8
 concurrencyOverrides {resolutionRule=PreventNew}
}
