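# Sound cue for picking up the Zapper Trap V2 item in Athena.
# A single pickup wave plays through a short fixed delay into a Modulator root.
# It uses the shared weapon pickup attenuation and the gadgets sound class, at reduced volume.
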SoundCue /Game/Athena/Sounds/Items/Zapper_Trap_V2/Zapper_Trap_V2_Pickup_Cue
root=Modulator_0
nodes [
 {id=Delay_0, children [WavePlayer_0], properties {delayMax=0.1, delayMin=0.1}}
 {id=Modulator_0, children [Delay_0]}
 {id=WavePlayer_0, properties {soundWaveAssetPtr=&/WPN_Trap_Zapper_V2_Pickup_01}}
]
properties {
 attenuationSettings=&Fort_Weapon_Pickup_Att
 maxDistance=3500
 soundClassObject=/Game/Sounds/FortSoundClassesAndModes/Fort_Gadgets_StW
 volumeMultiplier=0.3
}
