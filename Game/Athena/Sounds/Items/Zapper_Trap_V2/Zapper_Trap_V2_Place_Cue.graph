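# Sound cue for placing the Zapper Trap V2 in Athena, layering the trap's placement sound with its ready sound.
# A mixer plays both waves together, with the ready layer mixed quieter than the placement layer.
# It uses the Athena trap-activation attenuation and the weapon melee sound class.
# A per-owner concurrency rule limits it to three instances and prevents new plays beyond that.
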SoundCue /Game/Athena/Sounds/Items/Zapper_Trap_V2/Zapper_Trap_V2_Place_Cue
root=Mixer_8
nodes [
 {id=Mixer_8, children [WavePlayer_2,WavePlayer_1], properties {inputVolume [0.9,0.7]}}
 {id=WavePlayer_1, properties {soundWaveAssetPtr=&/WPN_Trap_Zapper_V2_Ready_01}}
 {id=WavePlayer_2, properties {soundWaveAssetPtr=&/WPN_Trap_Zapper_V2_Place_01}}
]
properties {
 attenuationSettings=/Game/Sounds/Attenuation/Fort_Traps_Activate_Athena_Att
 maxDistance=3500
 soundClassObject=&Fort_WeaponMelee_PSM
 volumeMultiplier=1
 attenuationOverrides {falloffDistance=2400, attenuationShapeExtents {x=600}}
 concurrencyOverrides {resolutionRule=PreventNew}
 concurrencySet [/Game/Sounds/Concurrency_Rules/ThreeMaximum_Owner_PreventNew]
}
