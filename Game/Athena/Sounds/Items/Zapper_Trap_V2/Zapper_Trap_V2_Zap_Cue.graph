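# Sound cue for the Zapper Trap V2 zap, pairing a firing sound with a trailing ready sound.
# A mixer layers a randomly picked, modulated fire variant (four equally weighted) with the ready sound, which plays after a fixed half-second delay.
# It uses a trap-activation attenuation asset, a third-person ranged-weapon fire sound class, and a concurrency override that prevents new instances once the limit is hit.
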SoundCue /Game/Athena/Sounds/Items/Zapper_Trap_V2/Zapper_Trap_V2_Zap_Cue
root=Mixer_19
nodes [
 {id=Delay_10, children [WavePlayer_4], properties {delayMax=0.5, delayMin=0.5}}
 {id=Mixer_19, children [Modulator_9,Delay_10], properties {inputVolume [0.75,0.7]}}
 {id=Modulator_9, children [Random_0]}
 {id=Random_0, children [WavePlayer_0,WavePlayer_1,WavePlayer_2,WavePlayer_3], properties {weights [1,1,1,1]}}
 {id=WavePlayer_0, properties {soundWaveAssetPtr=&/WPN_Trap_Zapper_V2_Fire_01}}
 {id=WavePlayer_1, properties {soundWaveAssetPtr=&/WPN_Trap_Zapper_V2_Fire_02}}
 {id=WavePlayer_2, properties {soundWaveAssetPtr=&/WPN_Trap_Zapper_V2_Fire_03}}
 {id=WavePlayer_3, properties {soundWaveAssetPtr=&/WPN_Trap_Zapper_V2_Fire_04}}
 {id=WavePlayer_4, properties {soundWaveAssetPtr=&/WPN_Trap_Zapper_V2_Ready_01}}
]
properties {
 attenuationSettings=/Game/Sounds/Attenuation/Fort_Traps_Activate_Athena_Att
 bOverrideConcurrency=true
 maxDistance=3500
 soundClassObject=&Fort_WeaponRanged_Fire_3P_PSM
 volumeMultiplier=1
 attenuationOverrides {falloffDistance=3200, attenuationShapeExtents {x=800}}
 concurrencyOverrides {resolutionRule=PreventNew}
}
